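# Niagara system for the SlickSwish emote that draws pink-red electric ribbons along bones of the player's skeletal mesh.
# One emitter spawns particles in ribbon-linked groups, which FN_BoneFlow positions across eight bone IDs with curved radius and random offsets.
# Curves over particle and emitter loop age feed per-layer colors and dynamic material parameters for two ribbon renderers, while ribbons narrow over life.
# Edge color dims at night, and a front-end preview flag lowers one layer's alpha.
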NiagaraSystem /BRCosmetics/Animation/Game/MainPlayer/Emotes/SlickSwish/FX/NS_Emote_SlickSwish_Electricity
dependencies {
 usedDataInterfaces [&NiagaraDataInterfaceCurve, &NiagaraDataInterfaceColorCurve]
 usedDynamicInputs []
 usedModules [&SystemState]
}
emitters {Bone_flow_idle_electricity {
  emitterUpdateScript [
   &EmitterState(Life Cycle Mode=::NewEnumerator1, Loop Duration=1, UseLoopDelay=false)
   &SpawnRate(SpawnRate=30, Use Spawn Probability=false)
   &SpawnBurst_Instantaneous(Use Spawn Probability=false, UseLoopCountLimit=false) {enabled=false}
  ]
  particleSpawnScript [
   &FN_SetSpawnGroup(Inclusivity=true, Normalize Spawn Group=true, Particles Per Group=6, Pattern=::NewEnumerator0, Set Ribbon ID and Ribbon Link Order=true)
   &InitializeParticle(
    Color=<rgba 1 0 0.1562 1>
    Lifetime Max=1
    Lifetime Min=0.5
    Lifetime Mode=::NewEnumerator1
    Ribbon Width=20
    Ribbon Width Mode=::NewEnumerator1
    UsePositionOffset=false
    Write Material Random=true
   )
   &CameraOffset(Camera Offset Amount=15)
  ]
  particleUpdateScript [
   &ParticleState()
   &FN_BoneFlow(
    Angle:NiagaraFloat=&Multiply_Float(A=$Particles.Normalized Spawn Group, B=1)
    BoneID:NiagaraInt32=&ModuloInt(A=$Particles.SpawnGroup, B=8)
    Enable Position Offsets=true
    Height:NiagaraFloat=&Multiply_Float(A=$Particles.Normalized Spawn Group, B=&RandomRangeFloat(Minimum=0.7, Maximum=1.11))
    Offset Mid:Vector3f=&RandomRangeVector(Minimum=<-1 -1 0>, Maximum=<1 1 0>)
    Radius:NiagaraFloat=&Multiply_Float(
     A=&FloatFromCurve(
      FloatCurve=DataInterface() {propertyValues {Curve {keys [
          <at 0 3 RCIM_Cubic>
          {interpMode=RCIM_Cubic, tangentMode=RCTM_User, time=0.0945, value=9.5896}
          {arriveTangent=-0.0483, interpMode=RCIM_Cubic, leaveTangent=-0.0483, time=0.4976, value=10.0326}
          {interpMode=RCIM_Cubic, tangentMode=RCTM_User, time=0.898, value=9.5507}
          <at 1 3>
         ]}}}
      CurveIndex=$Particles.Normalized Spawn Group
      Scale Curve=0.6
     )
     B=1
    )
    SkeletalMesh:NiagaraDataInterfaceSkeletalMesh=$User.Skeletal Mesh
   )
   /SetVariables(
    Particles.Dynamic params side lightning:Vector4f=&MakeVector4(
     X=0
     Y=&FloatFromCurve(FloatCurve=DataInterface() {propertyValues {Curve {keys [
          {arriveTangent=0.2955, interpMode=RCIM_Cubic, leaveTangent=0.2955, tangentMode=RCTM_User, time=0, value=0}
          {arriveTangent=-0.012, interpMode=RCIM_Cubic, leaveTangent=-0.012, tangentMode=RCTM_User, time=1, value=-1}
         ]}}})
     Z=0
     W=0
    )
    Particles.side lightning:LinearColor=&MakeLinearColorFromVectorAndFloat(
     "Vector (RGB)"=&MakeVectorFromLinearColorRGB(Color=&ColorFromCurve(ColorCurve=DataInterface() {propertyValues {
         AlphaCurve {keys []}
         BlueCurve {keys [<at 0 9.1543 RCIM_Cubic>,<at 0.3892 0.3449 RCIM_Cubic>]}
         GreenCurve {keys [<at 0 1 RCIM_Cubic>,<at 0.3892 0.0265 RCIM_Cubic>]}
         RedCurve {keys [<at 0 20 RCIM_Cubic>,<at 0.3892 0.6198 RCIM_Cubic>]}
        }}))
     "Float (Alpha)"=&FloatFromCurve(
      FloatCurve=DataInterface() {propertyValues {Curve {keys [
          {arriveTangent=-0.704, interpMode=RCIM_Cubic, leaveTangent=-0.704, tangentMode=RCTM_User, time=0, value=1}
          {interpMode=RCIM_Cubic, tangentMode=RCTM_User, time=1, value=0}
         ]}}}
      Scale Curve=&Multiply_Float(
       A=1
       B=&FloatFromCurve(FloatCurve=DataInterface() {propertyValues {Curve {keys [<at 0 0>,<at 0.1 0>,<at 0.2 1>,<at 1 1>]}}}, CurveIndex=$Emitter.NormalizedLoopAge)
      )
     )
    )
   ) {isSetParameters=true}
   &ScaleColor(
    Scale Alpha=&FloatFromCurve(
     FloatCurve=DataInterface() {propertyValues {Curve {keys [
         {arriveTangent=-1.0074, interpMode=RCIM_Cubic, leaveTangent=-1.0074, tangentMode=RCTM_User, time=0, value=1}
         {interpMode=RCIM_Cubic, tangentMode=RCTM_User, time=1, value=0}
        ]}}}
     Scale Curve=&Multiply_Float(
      A=4
      B=&FloatFromCurve(FloatCurve=DataInterface() {propertyValues {Curve {keys [<at 0 0>,<at 0.1 0>,<at 0.2 1>,<at 1 1>]}}}, CurveIndex=$Emitter.NormalizedLoopAge)
     )
    )
    Scale RGB=&Multiply_VectorByFloat(
     Vector=&MakeVectorFromLinearColorRGB(Color=<rgba 1 0.02 0.1731 1>)
     Float=&FNTimeOfDayScalar(Night Scale=0.5, Morning Scale=1, Evening Scale=1)
    )
    ScaleA=true
    ScaleRGB=true
   ) {enabled=false}
   &SolveForcesAndVelocity(Clamp Velocity=false, Limit Acceleration=false)
   &DynamicMaterialParameters(Module.Index0Float0Write=true, Module.Index0Float1Write=true, Module.Index0Float2Write=true, Module.Index0Float3Write=true)
   /SetVariables(Particles.Lightning center glow color:LinearColor=&MakeLinearColorFromVectorAndFloat(
     "Vector (RGB)"=&Multiply_VectorByFloat(
      Vector=&MakeVectorFromLinearColorRGB(Color=<rgba 1 0 0.1562 1>)
      Float=&Multiply_Float(A=1.5, B=1)
     )
     "Float (Alpha)"=&Multiply_Float(
      A=&FloatFromCurve(
       FloatCurve=DataInterface() {propertyValues {Curve {keys [<at 0 0>,<at 0.1 0>,<at 0.2 1>,<at 1 1>]}}}
       CurveIndex=$Emitter.NormalizedLoopAge
       Scale Curve=1.5
      )
      B=&FloatFromCurve(FloatCurve=DataInterface() {propertyValues {Curve {keys [
           <at 0 1>
           {arriveTangent=1.3985, time=0.2, value=1}
           {arriveTangent=-2, interpMode=RCIM_Cubic, leaveTangent=-2, time=0.4, value=1}
           {arriveTangent=-1.3423, interpMode=RCIM_Cubic, leaveTangent=-1.3423, tangentMode=RCTM_User, time=0.7, value=0}
          ]}}})
     )
    )) {isSetParameters=true}
   /SetVariables(
    Particles.Lightning center edge color:LinearColor=&MakeLinearColorFromVectorAndFloat(
     "Vector (RGB)"=&Multiply_VectorByFloat(
      Vector=&MakeVectorFromLinearColorRGB(Color=<rgba 1 0.02 0.1731 1>)
      Float=&FNTimeOfDayScalar(Night Scale=0.5, Morning Scale=1, Evening Scale=1)
     )
     "Float (Alpha)"=&FloatFromCurve(
      FloatCurve=DataInterface() {propertyValues {Curve {keys [
          {arriveTangent=-1.0074, interpMode=RCIM_Cubic, leaveTangent=-1.0074, tangentMode=RCTM_User, time=0, value=1}
          {interpMode=RCIM_Cubic, tangentMode=RCTM_User, time=1, value=0}
         ]}}}
      Scale Curve=&Multiply_Float(
       A=4
       B=&FloatFromCurve(FloatCurve=DataInterface() {propertyValues {Curve {keys [<at 0 0>,<at 0.1 0>,<at 0.2 1>,<at 1 1>]}}}, CurveIndex=$Emitter.NormalizedLoopAge)
      )
     )
    )
    Particles.Lightning center inside color:LinearColor=&MakeLinearColorFromVectorAndFloat(
     "Vector (RGB)"=&Multiply_VectorByFloat(
      Vector=&MakeVectorFromLinearColorRGB(Color=<rgba 0.9198 0.0262 0.3975 1>)
      Float=&Multiply_Float(
       A=10
       B=&FloatFromCurve(
        FloatCurve=DataInterface() {propertyValues {Curve {keys [<at 0 50>,<at 0.3 1>,<at 1 1>]}}}
        Scale Curve=&RandomRangeFloat(Minimum=0, Maximum=1, Evaluation Type=::NewEnumerator1)
       )
      )
     )
     "Float (Alpha)"=&Multiply_Float(
      A=&Multiply_Float(
       A=10
       B=&FloatFromCurve(FloatCurve=DataInterface() {propertyValues {Curve {keys [<at 0 0>,<at 0.1 0>,<at 0.2 0.8>,<at 1 0>]}}})
      )
      B=&FloatFromCurve(FloatCurve=DataInterface() {propertyValues {Curve {keys [<at 0 0>,<at 0.1 0>,<at 0.2 1>,<at 1 1>]}}}, CurveIndex=$Emitter.NormalizedLoopAge)
     )
    )
   ) {isSetParameters=true}
   /SetVariables(Particles.Color c:LinearColor=&MakeLinearColorFromVectorAndFloat(
     "Vector (RGB)"=&MakeVectorFromLinearColorRGB(Color=<rgba 1 1 1 1>)
     "Float (Alpha)"=&FloatFromCurve(
      FloatCurve=DataInterface() {propertyValues {Curve {keys [
          {arriveTangent=-1.178, interpMode=RCIM_Cubic, leaveTangent=-1.178, tangentMode=RCTM_User, time=0, value=1}
          {interpMode=RCIM_Cubic, tangentMode=RCTM_User, time=1, value=0}
         ]}}}
      Scale Curve=&MakeCustomFloatFromBool(Bool=$User.bIsFrontEndPreview, True Float=0.3, False Float=0.5)
     )
    )) {isSetParameters=true}
   /SetVariables(Particles.mask color:LinearColor=&MakeLinearColorFromVectorAndFloat(
     "Vector (RGB)"=&MakeVectorFromLinearColorRGB(Color=&ColorFromCurve(ColorCurve=DataInterface() {propertyValues {
         AlphaCurve {keys []}
         BlueCurve {keys [<at 0.2736 0.1 RCIM_Cubic>]}
         GreenCurve {keys [<at 0.2736 0.1 RCIM_Cubic>]}
         RedCurve {keys [<at 0.2736 0.1 RCIM_Cubic>]}
        }}))
     "Float (Alpha)"=&FloatFromCurve(
      FloatCurve=DataInterface() {propertyValues {Curve {keys [
          {arriveTangent=-1.2557, interpMode=RCIM_Cubic, leaveTangent=-1.2557, tangentMode=RCTM_User, time=0, value=1}
          {interpMode=RCIM_Cubic, tangentMode=RCTM_User, time=1, value=0}
         ]}}}
      Scale Curve=&Multiply_Float(
       A=1.5
       B=&FloatFromCurve(
        FloatCurve=DataInterface() {propertyValues {Curve {keys [<at 0 0>,<at 0.1 0>,<at 0.2 1>,<at 0.4 0.6>,<at 1 0>]}}}
        CurveIndex=$Emitter.NormalizedLoopAge
       )
      )
     )
    )) {isSetParameters=true}
   &DynamicMaterialParameters(
    Index 0 Param 1=&FloatFromCurve(FloatCurve=DataInterface() {propertyValues {Curve {keys [{tangentMode=RCTM_User, time=0, value=1}, {tangentMode=RCTM_User, time=1, value=1}]}}})
    Index 0 Param 2=0.7
    Index 0 Param 3=&FloatFromCurve(FloatCurve=DataInterface() {propertyValues {Curve {keys [
         {arriveTangent=1.0359, interpMode=RCIM_Cubic, leaveTangent=1.0359, tangentMode=RCTM_User, time=0, value=0}
         {arriveTangent=-0.7392, interpMode=RCIM_Cubic, leaveTangent=-0.7392, tangentMode=RCTM_User, time=1, value=3}
        ]}}})
    Index 1 Param 1=0.2
    Index 1 Param 2=&RandomRangeFloat(Minimum=0, Maximum=1)
    Index 1 Param 3=0.8
    Index 1 Param 4=1
    Index 2 Param 1=&FloatFromCurve(FloatCurve=DataInterface() {propertyValues {Curve {keys [<at 0 1>,<at 1 0.3>]}}})
    Module.Index0Float0Write=true
    Module.Index0Float1Write=true
    Module.Index0Float2Write=true
    Module.Index0Float3Write=true
    Module.Index1Float0Write=true
    Module.Index1Float1Write=true
    Module.Index1Float2Write=true
    Module.Index1Float3Write=true
    Module.Index2Float0Write=true
    Module.Index2Float1Write=false
    Module.Index2Float2Write=false
    Module.Index2Float3Write=false
    Write Parameter Index 1=::NewEnumerator1
    Write Parameter Index 2=::NewEnumerator1
   ) {module=DynamicMaterialParameters001}
   /SetVariables(
    Particles.Dynamic Parameters center lightning 1:Vector4f=&MakeVector4(
     X=&FloatFromCurve(FloatCurve=DataInterface() {propertyValues {Curve {keys [{tangentMode=RCTM_User, time=0, value=1}, {tangentMode=RCTM_User, time=1, value=1}]}}})
     Y=0.7
     Z=&FloatFromCurve(FloatCurve=DataInterface() {propertyValues {Curve {keys [
          {arriveTangent=1.0359, interpMode=RCIM_Cubic, leaveTangent=1.0359, tangentMode=RCTM_User, time=0, value=0}
          {arriveTangent=-0.7392, interpMode=RCIM_Cubic, leaveTangent=-0.7392, tangentMode=RCTM_User, time=1, value=3}
         ]}}})
     W=0
    )
    Particles.Dynamic Parameters center lightning 2:Vector4f=&MakeVector4(X=0.2, Y=&RandomRangeFloat(Minimum=0, Maximum=1), Z=0.8, W=1)
    Particles.Dynamic Parameters center lightning 3:Vector4f=&MakeVector4(
     X=&FloatFromCurve(FloatCurve=DataInterface() {propertyValues {Curve {keys [<at 0 1>,<at 1 0.3>]}}})
     Y=0
     Z=0
     W=0
    )
   ) {isSetParameters=true}
   /SetVariables(
    Particles.Glow color:LinearColor=&MakeLinearColorFromVectorAndFloat(
     "Vector (RGB)"=&MakeVectorFromLinearColorRGB(Color=&ColorFromCurve(ColorCurve=DataInterface() {propertyValues {
         AlphaCurve {keys []}
         BlueCurve {keys [<at 0.0511 2.0972 RCIM_Cubic>,<at 0.7078 0.3449 RCIM_Cubic>]}
         GreenCurve {keys [<at 0.0511 0.6667 RCIM_Cubic>,<at 0.7078 0.0265 RCIM_Cubic>]}
         RedCurve {keys [<at 0.0511 4 RCIM_Cubic>,<at 0.7078 0.6198 RCIM_Cubic>]}
        }}))
     "Float (Alpha)"=&FloatFromCurve(
      FloatCurve=DataInterface() {propertyValues {Curve {keys [
          {arriveTangent=-0.704, interpMode=RCIM_Cubic, leaveTangent=-0.704, tangentMode=RCTM_User, time=0, value=1}
          {interpMode=RCIM_Cubic, tangentMode=RCTM_User, time=1, value=0}
         ]}}}
      Scale Curve=&Multiply_Float(
       A=1
       B=&FloatFromCurve(
        FloatCurve=DataInterface() {propertyValues {Curve {keys [<at 0 0>,<at 0.1 0>,<at 0.2 1>,<at 0.7762 0>]}}}
        CurveIndex=$Emitter.NormalizedLoopAge
       )
      )
     )
    )
    Particles.dynamic parameters for glow 1:Vector4f=&MakeVector4(X=0.65, Y=0.5, Z=1, W=1)
   ) {isSetParameters=true}
   /SetVariables(Particles.Dy2:Vector4f=&MakeVector4(
     X=&FloatFromCurve(FloatCurve=DataInterface() {propertyValues {Curve {keys [{tangentMode=RCTM_User, time=0, value=0.2}, {tangentMode=RCTM_User, time=1, value=1}]}}})
     Y=0.7
     Z=0
     W=1
    )) {isSetParameters=true}
   &ScaleRibbonWidth(Ribbon Width Scale=&FloatFromCurve(FloatCurve=DataInterface() {propertyValues {Curve {
        defaultValue=1
        keys [
         {arriveTangent=0.0733, interpMode=RCIM_Cubic, leaveTangent=0.0733, tangentMode=RCTM_User, time=0, value=1}
         {arriveTangent=-0.0405, interpMode=RCIM_Cubic, leaveTangent=-0.0405, tangentMode=RCTM_User, time=0.7, value=0.3}
        ]
       }}}))
  ]
  rendererClasses [&NiagaraRibbonRendererProperties]
  renderers [
   {rendererClass=&NiagaraRibbonRendererProperties}
   {rendererClass=&NiagaraRibbonRendererProperties, rendererIndex=1}
  ]
 }}
summary {
 systemName=NS_Emote_SlickSwish_Electricity
 emitters [{emitterName=Bone_flow_idle_electricity, rendererClasses [&NiagaraRibbonRendererProperties]}]
 userVariables [
  User.bIsFrontEndPreview {type=NiagaraBool}
  User.Skeletal Mesh {
   type=NiagaraDataInterfaceSkeletalMesh
   defaultValue {
    struct=NiagaraExt_VariableValue_DataInterface
    value {dataInterface=&self:NiagaraDataInterfaceSkeletalMesh_3}
   }
  }
 ]
}
