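# Niagara effect for the SlickSwish emote that puts pink-magenta electricity and a glowing ribbon trail on one of the player's hands.
# Mesh arcs with flickering particle lights follow hand sockets, while a second, sparser mesh emitter leaves drifting arcs behind.
# A looping anim-trail emitter draws layered ribbons from hand bones, with User anim-curve floats scaling electricity rate, ribbon lifetime and width.
# Time-of-day scalars adjust glow and alpha.
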NiagaraSystem /BRCosmetics/Animation/Game/MainPlayer/Emotes/SlickSwish/FX/NS_Emote_SlickSwish_SingleHand_trail
dependencies {
 usedDataInterfaces [&NiagaraDataInterfaceCurve, &NiagaraDataInterfaceArrayFloat3, &NiagaraDataInterfaceColorCurve]
 usedDynamicInputs []
 usedModules [
  &self:SystemSpawnScript.SystemScriptSource.SystemScriptGraph.NiagaraNodeAssignment_0.SetVariables_#1
  &SystemState
 ]
}
emitters {
 Hand_electricity001 {
  emitterUpdateScript [
   &EmitterState()
   &SpawnRate(SpawnGroup=1, Use Spawn Probability=false) {enabled=false, module=SpawnRate001}
   &SpawnRate(SpawnRate=&Multiply_Float(A=35, B=$User.Anim_Curve_SpawnRate_electricity), Use Spawn Probability=true)
  ]
  particleSpawnScript [
   &FN_SetSpawnGroup()
   &PartitionParticles()
   &InitializeParticle(
    Color=<rgba 50 0 7.8112 1>
    Lifetime Max=0.35
    Lifetime Min=0.1
    Lifetime Mode=::NewEnumerator1
    Mesh Scale=&RandomRangeVector(Minimum=<0.085 0.085 0.085>, Maximum=<0.16 0.16 0.16>)
    Mesh Scale Mode=::NewEnumerator4
    Sprite Size Mode=::NewEnumerator1
    Uniform Sprite Size Max=1
    Uniform Sprite Size Min=0.2
    UsePositionOffset=false
    Write Material Random=true
   )
   /SetVariables(Particles.Mesh index:NiagaraInt32=&RandomRangeInt(Minimum=0, Maximum=2)) {isSetParameters=true, module=SetVariables_#2}
  ]
  particleUpdateScript [
   /SetVariables(Particles.BrightnessScale:NiagaraFloat=&FloatFromCurve(
     FloatCurve=DataInterface() {propertyValues {Curve {keys [<at 0 50>,<at 0.3 1>,<at 1 1>]}}}
     Scale Curve=&RandomRangeFloat(Minimum=0, Maximum=1, Evaluation Type=::NewEnumerator1)
    )) {enabled=false, isSetParameters=true}
   &ParticleState()
   &SkeletalMeshLocation(
    Bone / Socket Index=$Particles.PartitionParticles.Partition
    MaskSpawnGroup=false
    Mesh Sampling Type=::NewEnumerator5
    Orientation Sampling=::NewEnumerator1
    Sampled Position Offset=&SelectVectorFromArray(
     Direct Array Index=$Particles.PartitionParticles.Partition
     Vector Selection Array=DataInterface() {propertyValues {FloatData [<-0.5263 0 0>, <0 0 0>]}}
    )
    Skeletal Mesh=$System.Skeletal Mesh
    Socket Sampling Mode=::NewEnumerator5
   )
   &ShapeLocation(
    Cylinder Height=8
    Cylinder Radius=5
    MaskSpawnGroup=false
    Override Seed=false
    Shape Origin=$Particles.SkeletalMeshLocation.SampledPosition
    Shape Primitive=::NewEnumerator1
    Surface Only=false
    UseOwnerScale=true
   )
   &ScaleColor(
    Scale Alpha=&Waveform(
     "Waveform [1]"=::NewEnumerator0
     "[1] Frequency"=&RandomRangeFloat(Minimum=5, Maximum=8)
     "[1] Amplitude Min/Max"=<0.5 1>
    )
    ScaleA=true
    ScaleRGB=true
   ) {enabled=false}
   &CameraOffset(Camera Offset Amount=3)
   /SetVariables(Particles.Glow color:LinearColor=&MakeLinearColorFromVectorAndFloat(
     "Vector (RGB)"=&Multiply_VectorByFloat(Vector=&MakeVectorFromLinearColorRGB(Color=<rgba 1 0 0.1562 1>), Float=1)
     "Float (Alpha)"=&Multiply_Float(
      A=&Multiply_Float(
       A=&FloatFromCurve(
        FloatCurve=DataInterface() {propertyValues {Curve {keys [
            {interpMode=RCIM_Cubic, tangentMode=RCTM_User, time=0, value=0}
            {interpMode=RCIM_Cubic, tangentMode=RCTM_User, time=0.5, value=1}
            {tangentMode=RCTM_User, time=1, value=0}
           ]}}}
        Scale Curve=&FNTimeOfDayScalar(Night Scale=5, Morning Scale=0.5, Day Scale=5, Evening Scale=0.5)
       )
       B=&Multiply_Float(A=&RandomRangeFloat(Minimum=0.5, Maximum=1, Evaluation Type=::NewEnumerator1), B=1)
      )
      B=1
     )
    )) {isSetParameters=true, module=SetVariables_#3}
   &ScaleColor(
    Scale Alpha=&Multiply_Float(
     A=&FloatFromCurve(
      FloatCurve=DataInterface() {propertyValues {Curve {keys [
          {interpMode=RCIM_Cubic, tangentMode=RCTM_User, time=0, value=0}
          {interpMode=RCIM_Cubic, tangentMode=RCTM_User, time=0.5, value=1}
          {tangentMode=RCTM_User, time=0.85, value=0}
         ]}}}
      Scale Curve=&FNTimeOfDayScalar(Night Scale=1, Morning Scale=0.5, Evening Scale=0.5)
     )
     B=&Multiply_Float(A=&RandomRangeFloat(Minimum=0.5, Maximum=1, Evaluation Type=::NewEnumerator1), B=1)
    )
    Scale RGB=<10 10 10>
    ScaleA=true
    ScaleRGB=true
   ) {module=ScaleColor001}
   /SetVariables(Particles.Glow Dy02:Vector4f=&MakeVector4(
     X=&Multiply_Float(
      A=2
      B=&FloatFromCurve(FloatCurve=DataInterface() {propertyValues {Curve {keys [
           {arriveTangent=-2.6097, interpMode=RCIM_Cubic, leaveTangent=-2.6097, tangentMode=RCTM_User, time=0, value=1.5}
           {arriveTangent=-0.6794, interpMode=RCIM_Cubic, leaveTangent=-0.6794, tangentMode=RCTM_User, time=1, value=0}
          ]}}})
     )
     Y=0.01
     Z=0
     W=0
    )) {isSetParameters=true, module=SetVariables_#4}
   /SetVariables(Particles.MeshOrientation=&MakeQuaternion(
     Quaternion From=5
     Axis=?
     Angle=?
     Yaw=?
     Pitch=?
     Roll=?
     X=?
     Y=?
     Z=?
     From=?
     To=?
     Angle Type=::NewEnumerator0
     Coordinate Space=?
     Normalize Vectors=?
    )) {isSetParameters=true}
   &CameraOffset(Camera Offset Amount=15) {module=CameraOffset001}
   &Color(Color=<rgba 1 0 0.1562 1>, Scale Alpha=5)
   &DynamicMaterialParameters(
    Index 0 Param 1=&MakeFloatFromVector(Vector=$Particles.Scale, Channel=::NewEnumerator2)
    Index 0 Param 2=&RandomRangeFloat(Minimum=0.3, Maximum=10)
    Index 0 Param 3=&FloatFromCurve(
     FloatCurve=DataInterface() {propertyValues {Curve {keys [{interpMode=RCIM_Cubic, leaveTangent=-2.6584, tangentMode=RCTM_Break, time=0.8568, value=1}, {tangentMode=RCTM_User, time=1, value=0}]}}}
     Scale Curve=0.02
    )
    Index 0 Param 4=1
    Index 1 Param 1=&FloatFromCurve(FloatCurve=DataInterface() {propertyValues {Curve {keys [
         {interpMode=RCIM_Cubic, leaveTangent=1.1034, tangentMode=RCTM_Break, time=0, value=0}
         {arriveTangent=2.2399, tangentMode=RCTM_User, time=0.5, value=1}
        ]}}})
    Index 1 Param 2=&FloatFromCurve(FloatCurve=DataInterface() {propertyValues {Curve {keys [<at 0.5 0>, {arriveTangent=3.2053, tangentMode=RCTM_User, time=0.9, value=1}]}}})
    Index 1 Param 4=&FloatFromCurve(
     FloatCurve=DataInterface() {propertyValues {Curve {keys [
         {interpMode=RCIM_Cubic, leaveTangent=0.1169, tangentMode=RCTM_Break, time=0, value=0}
         {arriveTangent=2.5708, interpMode=RCIM_Cubic, leaveTangent=2.5708, tangentMode=RCTM_User, time=1, value=1}
        ]}}}
     Scale Curve=&RandomRangeFloat(Minimum=30, Maximum=40)
    )
    Index 2 Param 1=&FloatFromCurve(FloatCurve=DataInterface() {propertyValues {Curve {keys [
         {interpMode=RCIM_Cubic, leaveTangent=-0.0025, tangentMode=RCTM_Break, time=0, value=0.5}
         {arriveTangent=-0.0316, interpMode=RCIM_Cubic, tangentMode=RCTM_Break, time=1, value=-0.2}
        ]}}})
    Module.Index0Param0Write=true
    Module.Index0Param1Write=true
    Module.Index0Param2Write=true
    Module.Index0Param3Write=true
    Module.Index1Param0Write=true
    Module.Index1Param1Write=true
    Module.Index1Param2Write=true
    Module.Index1Param3Write=true
    Module.Index2Param0Write=true
    Module.Index2Param1Write=true
    Module.Index2Param2Write=true
    Module.Index2Param3Write=true
    Module.Index3Param0Write=true
    Module.Index3Param1Write=true
    Module.Index3Param2Write=true
    Module.Index3Param3Write=true
    Write Parameter Index 1=true
    Write Parameter Index 2=true
    Write Parameter Index 3=true
   )
   &Drag(
    Drag=&MaskFloat(
     Mask Source=0
     Mask Parameter=?
     Mask=0
     Masked Float=0
     Passthrough Float=&FloatFromCurve(FloatCurve=DataInterface() {propertyValues {Curve {
         defaultValue=1
         keys [{interpMode=RCIM_Cubic, tangentMode=RCTM_User, time=0, value=1}, {interpMode=RCIM_Cubic, tangentMode=RCTM_User, time=0.5582, value=0}]
        }}})
    )
    Use Linear Drag=true
    Use Rotational Drag=true
   )
   &GravityForce(Gravity=&MaskVector(Mask Source=0, Mask Parameter=?, Mask=0, Masked Vector=<0 0 1500>, Passthrough Vector=<1 1 1500>)) {enabled=false}
   &SolveForcesAndVelocity(Clamp Velocity=false, Limit Acceleration=false)
   &FN_Light_Attributes(
    Color=<rgba 1 0 0.1562 1>
    Create Color Attribute=true
    Create Radius Attribute=true
    Enable Particle Light Enable=true
    Radius=&FloatFromCurve(
     FloatCurve=DataInterface() {propertyValues {Curve {keys [<at 0 1>,<at 0.1959 0.3345>,<at 0.4134 0.677>,<at 0.6308 0.2598>,<at 0.7381 0.8094>,<at 0.8085 0.0094>]}}}
     Scale Curve=30
    )
    Scale Color=&FloatFromCurve(FloatCurve=DataInterface() {propertyValues {Curve {keys [<at 0 1>,<at 0.5 0>]}}})
   )
  ]
  rendererClasses [&NiagaraMeshRendererProperties, &NiagaraLightRendererProperties]
  renderers [
   {rendererClass=&NiagaraMeshRendererProperties, rendererIndex=0}
   {rendererClass=&NiagaraMeshRendererProperties, rendererIndex=1}
   {rendererClass=&NiagaraLightRendererProperties, rendererIndex=2}
  ]
 }
 Hand_electricity002_left_behind {
  emitterUpdateScript [
   &EmitterState()
   &SpawnRate(SpawnGroup=1, Use Spawn Probability=false) {enabled=false, module=SpawnRate001}
   &SpawnRate(SpawnRate=&Multiply_Float(A=15, B=1), Use Spawn Probability=true)
  ]
  particleSpawnScript [
   &PartitionParticles(Number Of Partitions=1)
   &InitializeParticle(
    Color=<rgba 50 0 7.8112 1>
    Lifetime Max=0.3
    Lifetime Min=0.1
    Lifetime Mode=::NewEnumerator1
    Mesh Scale=&RandomRangeVector(Minimum=<0.1 0.1 0.1>, Maximum=<0.2 0.2 0.5>)
    Mesh Scale Mode=::NewEnumerator4
    Sprite Size Mode=::NewEnumerator1
    Uniform Sprite Size Max=1.5
    Uniform Sprite Size Min=0.7
    UsePositionOffset=false
    Write Material Random=true
   )
   /SetVariables() {isSetParameters=true, module=SetVariables_#2}
   &AddVelocity(Velocity=<0 0 613.4774>) {enabled=false}
   &AddVelocity(
    Module.Use Falloff Distance=false
    Origin Offset=<-50 0 -3>
    Origin Offset Coordinate Space=::World
    Override Seed=false
    Velocity Mode=::NewEnumerator1
    Velocity Speed=&Multiply_Float(
     A=&RandomRangeFloat(Minimum=100, Maximum=500)
     B=&FloatFromCurve(
      FloatCurve=DataInterface() {propertyValues {Curve {keys [{interpMode=RCIM_Cubic, tangentMode=RCTM_User, time=0, value=0.1585}, {interpMode=RCIM_Cubic, tangentMode=RCTM_User, time=0.0808, value=1}]}}}
      CurveIndex=$Emitter.Age
     )
    )
   ) {enabled=false, module=AddVelocity001}
   &SkeletalMeshLocation(
    Bone / Socket Index=$Particles.PartitionParticles.Partition
    Bone Sampling Mode=::NewEnumerator4
    "Inherit Velocity (Scale)"=0.1
    MaskSpawnGroup=false
    Sampled Position Offset=&MaskVector(Mask Source=1, Mask Parameter=$Particles.PartitionParticles.Partition, Mask=0, Masked Vector=<0 -4.8987 0>, Passthrough Vector=<0 0 0>)
    Skeletal Mesh=$System.Skeletal Mesh
    Velocity Sampling=::NewEnumerator0
    VelocitySpeedLimit=false
   ) {module=SkeletalMeshLocation001}
   &ShapeLocation(MaskSpawnGroup=false, Offset Mode=::NewEnumerator0, Override Seed=false, Sphere Radius=3, UseOwnerScale=true) {module=ShapeLocation001}
  ]
  particleUpdateScript [
   &ParticleState()
   &SkeletalMeshLocation(
    Bone / Socket Index=$Particles.PartitionParticles.Partition
    MaskSpawnGroup=false
    Mesh Sampling Type=::NewEnumerator5
    Orientation Sampling=::NewEnumerator1
    Sampled Position Offset=&SelectVectorFromArray(
     Direct Array Index=$Particles.PartitionParticles.Partition
     Vector Selection Array=DataInterface() {propertyValues {FloatData [<-0.5263 0 0>, <0 0 0>]}}
    )
    Skeletal Mesh=$System.Skeletal Mesh
    Socket Sampling Mode=::NewEnumerator5
   ) {enabled=false}
   &ShapeLocation(
    Cylinder Height=5
    Cylinder Radius=1
    MaskSpawnGroup=false
    Override Seed=false
    Shape Origin=$Particles.SkeletalMeshLocation.SampledPosition
    Shape Primitive=::NewEnumerator1
    Surface Only=false
    UseOwnerScale=true
   ) {enabled=false}
   &ScaleColor(
    Scale Alpha=&Waveform(
     "Waveform [1]"=::NewEnumerator0
     "[1] Frequency"=&RandomRangeFloat(Minimum=5, Maximum=8)
     "[1] Amplitude Min/Max"=<0.5 1>
    )
    ScaleA=true
    ScaleRGB=true
   )
   &CameraOffset(Camera Offset Amount=3)
   /SetVariables(Particles.Glow color:LinearColor=&MakeLinearColorFromVectorAndFloat(
     "Vector (RGB)"=&Multiply_VectorByFloat(Vector=&MakeVectorFromLinearColorRGB(Color=<rgba 1 0 0.1562 1>), Float=1)
     "Float (Alpha)"=&Multiply_Float(
      A=&Multiply_Float(
       A=&FloatFromCurve(
        FloatCurve=DataInterface() {propertyValues {Curve {keys [
            {interpMode=RCIM_Cubic, tangentMode=RCTM_User, time=0, value=0}
            {interpMode=RCIM_Cubic, tangentMode=RCTM_User, time=0.5, value=1}
            {tangentMode=RCTM_User, time=1, value=0}
           ]}}}
        Scale Curve=&FNTimeOfDayScalar(Night Scale=0.7, Morning Scale=0.3, Evening Scale=0.3)
       )
       B=&Multiply_Float(A=&RandomRangeFloat(Minimum=0, Maximum=1, Evaluation Type=::NewEnumerator1), B=1)
      )
      B=1
     )
    )) {isSetParameters=true, module=SetVariables_#3}
   &ScaleColor(
    Scale Alpha=&Multiply_Float(
     A=&FloatFromCurve(
      FloatCurve=DataInterface() {propertyValues {Curve {keys [
          {interpMode=RCIM_Cubic, tangentMode=RCTM_User, time=0, value=0}
          {interpMode=RCIM_Cubic, tangentMode=RCTM_User, time=0.5, value=1}
          {tangentMode=RCTM_User, time=0.85, value=0}
         ]}}}
      Scale Curve=&FNTimeOfDayScalar(Night Scale=0.7, Morning Scale=0.3, Evening Scale=0.3)
     )
     B=&Multiply_Float(A=&RandomRangeFloat(Minimum=0, Maximum=1, Evaluation Type=::NewEnumerator1), B=1)
    )
    ScaleA=true
    ScaleRGB=true
   ) {module=ScaleColor001}
   /SetVariables(Particles.Glow Dy02:Vector4f=&MakeVector4(
     X=&Multiply_Float(
      A=2
      B=&FloatFromCurve(FloatCurve=DataInterface() {propertyValues {Curve {keys [
           {arriveTangent=-2.6097, interpMode=RCIM_Cubic, leaveTangent=-2.6097, tangentMode=RCTM_User, time=0, value=1.5}
           {arriveTangent=-0.6794, interpMode=RCIM_Cubic, leaveTangent=-0.6794, tangentMode=RCTM_User, time=1, value=0}
          ]}}})
     )
     Y=0
     Z=0
     W=0
    )) {isSetParameters=true, module=SetVariables_#4}
   &Color(Color=<rgba 1 0 0.1562 1>)
   &DynamicMaterialParameters(
    Index 0 Param 1=&MakeFloatFromVector(Vector=$Particles.Scale, Channel=::NewEnumerator2)
    Index 0 Param 2=&RandomRangeFloat(Minimum=0, Maximum=10)
    Index 0 Param 3=&FloatFromCurve(
     FloatCurve=DataInterface() {propertyValues {Curve {keys [{interpMode=RCIM_Cubic, leaveTangent=-2.6584, tangentMode=RCTM_Break, time=0.8568, value=1}, {tangentMode=RCTM_User, time=1, value=0}]}}}
     Scale Curve=0.02
    )
    Index 0 Param 4=1
    Index 1 Param 1=&FloatFromCurve(FloatCurve=DataInterface() {propertyValues {Curve {keys [
         {interpMode=RCIM_Cubic, leaveTangent=1.1034, tangentMode=RCTM_Break, time=0, value=0}
         {arriveTangent=2.2399, tangentMode=RCTM_User, time=1, value=1}
        ]}}})
    Index 1 Param 2=&FloatFromCurve(FloatCurve=DataInterface() {propertyValues {Curve {keys [<at 0.5 0>, {arriveTangent=3.2053, tangentMode=RCTM_User, time=0.9, value=1}]}}})
    Index 1 Param 4=&FloatFromCurve(
     FloatCurve=DataInterface() {propertyValues {Curve {keys [
         {interpMode=RCIM_Cubic, leaveTangent=0.1169, tangentMode=RCTM_Break, time=0, value=0}
         {arriveTangent=2.5708, interpMode=RCIM_Cubic, leaveTangent=2.5708, tangentMode=RCTM_User, time=1, value=1}
        ]}}}
     Scale Curve=&RandomRangeFloat(Minimum=30, Maximum=40)
    )
    Index 2 Param 1=&FloatFromCurve(FloatCurve=DataInterface() {propertyValues {Curve {keys [
         {interpMode=RCIM_Cubic, leaveTangent=-0.0025, tangentMode=RCTM_Break, time=0, value=0.5}
         {arriveTangent=-0.0316, interpMode=RCIM_Cubic, tangentMode=RCTM_Break, time=1, value=-0.2}
        ]}}})
    Module.Index0Param0Write=true
    Module.Index0Param1Write=true
    Module.Index0Param2Write=true
    Module.Index0Param3Write=true
    Module.Index1Param0Write=true
    Module.Index1Param1Write=true
    Module.Index1Param2Write=true
    Module.Index1Param3Write=true
    Module.Index2Param0Write=true
    Module.Index2Param1Write=true
    Module.Index2Param2Write=true
    Module.Index2Param3Write=true
    Module.Index3Param0Write=true
    Module.Index3Param1Write=true
    Module.Index3Param2Write=true
    Module.Index3Param3Write=true
    Write Parameter Index 1=true
    Write Parameter Index 2=true
    Write Parameter Index 3=true
   )
   &Drag(
    Drag=&MaskFloat(
     Mask Source=0
     Mask Parameter=?
     Mask=0
     Masked Float=0
     Passthrough Float=&FloatFromCurve(
      FloatCurve=DataInterface() {propertyValues {Curve {
         defaultValue=1
         keys [{interpMode=RCIM_Cubic, tangentMode=RCTM_User, time=0, value=1}, {interpMode=RCIM_Cubic, tangentMode=RCTM_User, time=0.5582, value=0}]
        }}}
      Scale Curve=0.5
     )
    )
    Use Linear Drag=true
    Use Rotational Drag=true
   )
   &GravityForce(Gravity=<0 0 15>)
   &SolveForcesAndVelocity(Clamp Velocity=false, Limit Acceleration=false)
   &FN_Light_Attributes(
    Color=<rgba 0.6867 0 0.5054 1>
    Create Color Attribute=true
    Create Radius Attribute=true
    Enable Particle Light Enable=true
    Light Enabled=true
    Radius=&FloatFromCurve(
     FloatCurve=DataInterface() {propertyValues {Curve {keys [<at 0 1>,<at 0.1959 0.3345>,<at 0.4134 0.677>,<at 0.6308 0.2598>,<at 0.8385 0.8188>,<at 1 0>]}}}
     Scale Curve=50
    )
    Scale Color=&FloatFromCurve(FloatCurve=DataInterface() {propertyValues {Curve {keys [<at 0 1>,<at 0.5 0>]}}})
   ) {enabled=false}
  ]
  rendererClasses [&NiagaraMeshRendererProperties]
  renderers [
   {rendererClass=&NiagaraMeshRendererProperties, rendererIndex=0}
   {rendererClass=&NiagaraMeshRendererProperties, rendererIndex=1}
  ]
 }
 hand_trail {
  emitterUpdateScript [
   &EmitterState(Life Cycle Mode=::NewEnumerator1, Loop Behavior=::NewEnumerator1, Loop Duration=0.65, UseLoopDelay=false)
   &FN_SpawnAnimTrail(AnimTrailWidthMode=::FromFirst, Spawn Interval=0, Spawn Mode:ENiagaraSkeletalSpawning=::NewEnumerator0)
  ]
  particleSpawnScript [
   &FN_InitializeAnimTrailBoneIndecies(AnimTrail Selection=1)
   &FN_InitializeAnimTrail(
    AnimTrailWidth=15
    Lifetime=&Multiply_Float(
     A=&FloatFromCurve(
      FloatCurve=DataInterface() {propertyValues {Curve {
         defaultValue=1
         keys [{interpMode=RCIM_Cubic, tangentMode=RCTM_User, time=0, value=1}, {interpMode=RCIM_Cubic, tangentMode=RCTM_User, time=0.5, value=-0.0601}]
        }}}
      Scale Curve=$User.Anim_Curve_SpawnRate_Ribbon
     )
     B=1
    )
    SkeletalMesh=$System.Skeletal Mesh
   )
   &SkeletalMeshLocation(
    Bone / Socket Index=1
    Bone Sampling Mode=::NewEnumerator4
    MaskSpawnGroup=false
    Sampled Position Offset=<-10 0 0>
    Skeletal Mesh=$System.Skeletal Mesh
   ) {module=SkeletalMeshLocation001}
   &AddVelocity(Velocity=<0 0 500>) {enabled=false}
   &ApplyInitialForces(Apply Force to Position=true)
   &CameraOffset(Camera Offset Amount=8)
  ]
  particleUpdateScript [
   &ParticleState()
   &InheritSourceMovement(ApplyForce=false, ApplyPosition=true, ApplyPositionFalloff=false)
   &ScaleRibbonWidth(Ribbon Width Scale=&FloatFromCurve(
     FloatCurve=DataInterface() {propertyValues {Curve {
        defaultValue=1
        keys [
         {arriveTangent=0.0733, interpMode=RCIM_Cubic, leaveTangent=0.0733, tangentMode=RCTM_User, time=0, value=1}
         {arriveTangent=-0.0405, interpMode=RCIM_Cubic, leaveTangent=-0.0405, tangentMode=RCTM_User, time=0.7, value=0.3}
        ]
       }}}
     Scale Curve=&Multiply_Float(A=1.3, B=$User.Anim_Curve_ribbon_width)
    ))
   /SetVariables(Particles.side lightning:LinearColor=&MakeLinearColorFromVectorAndFloat(
     "Vector (RGB)"=&MakeVectorFromLinearColorRGB(Color=&ColorFromCurve(ColorCurve=DataInterface() {propertyValues {
         AlphaCurve {keys []}
         BlueCurve {keys [<at 0.0511 9.1543 RCIM_Cubic>,<at 0.6302 0.3449 RCIM_Cubic>]}
         GreenCurve {keys [<at 0.0511 1 RCIM_Cubic>,<at 0.6302 0.0265 RCIM_Cubic>]}
         RedCurve {keys [<at 0.0511 20 RCIM_Cubic>,<at 0.6302 0.6198 RCIM_Cubic>]}
        }}))
     "Float (Alpha)"=&FloatFromCurve(
      FloatCurve=DataInterface() {propertyValues {Curve {keys [
          {arriveTangent=-0.704, interpMode=RCIM_Cubic, leaveTangent=-0.704, tangentMode=RCTM_User, time=0, value=1}
          {interpMode=RCIM_Cubic, tangentMode=RCTM_User, time=0.6, value=0}
         ]}}}
      Scale Curve=&Multiply_Float(
       A=1
       B=&FloatFromCurve(FloatCurve=DataInterface() {propertyValues {Curve {keys [<at 0 0>,<at 0.1 0>,<at 0.2 1>,<at 1 1>]}}}, CurveIndex=$Emitter.NormalizedLoopAge)
      )
     )
    )) {isSetParameters=true}
   &ScaleColor(
    Scale Alpha=&FloatFromCurve(FloatCurve=DataInterface() {propertyValues {Curve {keys [{arriveTangent=-3.0806, interpMode=RCIM_Cubic, leaveTangent=-3.0806, tangentMode=RCTM_User, time=0, value=1}, <at 0.85 0.2071 RCIM_Cubic>]}}})
    Scale RGB=&MakeVectorFromLinearColorRGB(Color=&ColorFromCurve(ColorCurve=DataInterface() {propertyValues {
        AlphaCurve {keys []}
        BlueCurve {keys [<at 0 0.2583>,<at 0.3816 0.5054 RCIM_Cubic>]}
        GreenCurve {keys [<at 0 0>,<at 0.3816 0 RCIM_Cubic>]}
        RedCurve {keys [<at 0 0.8>,<at 0.3816 0.6867 RCIM_Cubic>]}
       }}))
    ScaleA=true
    ScaleRGB=true
   ) {enabled=false}
   &ScaleColor(
    Scale Alpha=&FloatFromCurve(FloatCurve=DataInterface() {propertyValues {Curve {keys [{arriveTangent=-3.0806, interpMode=RCIM_Cubic, leaveTangent=-3.0806, tangentMode=RCTM_User, time=0, value=1}, <at 0.3 0 RCIM_Cubic>]}}})
    Scale RGB=&MakeVectorFromLinearColorRGB(Color=&ColorFromCurve(ColorCurve=DataInterface() {propertyValues {
        AlphaCurve {keys []}
        BlueCurve {keys [<at 0 0.2583>,<at 0.3816 0.5054 RCIM_Cubic>]}
        GreenCurve {keys [<at 0 0>,<at 0.3816 0 RCIM_Cubic>]}
        RedCurve {keys [<at 0 0.8>,<at 0.3816 0.6867 RCIM_Cubic>]}
       }}))
    ScaleA=true
    ScaleRGB=true
   ) {module=ScaleColor003}
   &ScaleColor(
    Scale Alpha=0.7
    Scale RGB=&VectorFromFloat(Value=&FNTimeOfDayScalar(Night Scale=85, Morning Scale=85, Day Scale=85, Evening Scale=85))
    ScaleA=false
    ScaleRGB=true
   ) {module=ScaleColor001}
   &ScaleColor(
    Scale Alpha=&FNTimeOfDayScalar(Night Scale=15, Morning Scale=1, Day Scale=1500, Evening Scale=1)
    Scale RGB=&VectorFromFloat(Value=&FNTimeOfDayScalar(Night Scale=100, Morning Scale=125, Day Scale=150, Evening Scale=85))
    ScaleA=false
    ScaleRGB=true
   ) {enabled=false, module=ScaleColor002}
   &Drag(Drag=5, Use Linear Drag=true, Use Rotational Drag=false)
   &SolveForcesAndVelocity(Clamp Velocity=false, Limit Acceleration=false)
   &DynamicMaterialParameters(
    Index 0 Param 1=1
    Module.Index0Float0Write=true
    Module.Index0Float1Write=true
    Module.Index0Float2Write=true
    Module.Index0Float3Write=true
   )
   &FN_UpdateAnimTrail(SkeletalMesh=$System.Skeletal Mesh)
   /SetVariables(
    Particles.Glow color:LinearColor=&MakeLinearColorFromVectorAndFloat(
     "Vector (RGB)"=&MakeVectorFromLinearColorRGB(Color=&ColorFromCurve(ColorCurve=DataInterface() {propertyValues {
         AlphaCurve {keys []}
         BlueCurve {keys [<at 0.0128 0.4155>,<at 0.3229 0.3449 RCIM_Cubic>]}
         GreenCurve {keys [<at 0.0128 0.1788>,<at 0.3229 0.0265 RCIM_Cubic>]}
         RedCurve {keys [<at 0.0128 0.6198>,<at 0.3229 0.6198 RCIM_Cubic>]}
        }}))
     "Float (Alpha)"=&FloatFromCurve(
      FloatCurve=DataInterface() {propertyValues {Curve {keys [
          {arriveTangent=-0.704, interpMode=RCIM_Cubic, leaveTangent=-0.704, tangentMode=RCTM_User, time=0, value=1}
          {interpMode=RCIM_Cubic, tangentMode=RCTM_User, time=1, value=0}
         ]}}}
      Scale Curve=&Multiply_Float(
       A=1
       B=&FloatFromCurve(
        FloatCurve=DataInterface() {propertyValues {Curve {keys [<at 0 0>,<at 0.1 0>,<at 0.2 1>,<at 0.7762 0>]}}}
        CurveIndex=$Emitter.NormalizedLoopAge
        Scale Curve=0.7
       )
      )
     )
    )
    Particles.dynamic parameters for glow 1:Vector4f=&MakeVector4(X=0.8, Y=1, Z=0, W=0)
   ) {isSetParameters=true}
   /SetVariables(Particles.Color B:LinearColor=&MakeLinearColorFromVectorAndFloat(
     "Vector (RGB)"=<5 5 5>
     "Float (Alpha)"=&FloatFromCurve(
      FloatCurve=DataInterface() {propertyValues {Curve {keys [
          {arriveTangent=-1.178, interpMode=RCIM_Cubic, leaveTangent=-1.178, tangentMode=RCTM_User, time=0, value=1}
          {interpMode=RCIM_Cubic, tangentMode=RCTM_User, time=0.6, value=0}
         ]}}}
      Scale Curve=&FNTimeOfDayScalar(Night Scale=0.15, Morning Scale=0.25, Day Scale=0.3, Evening Scale=0.2)
     )
    )) {isSetParameters=true}
  ]
  rendererClasses [&NiagaraRibbonRendererProperties]
  renderers [
   {rendererClass=&NiagaraRibbonRendererProperties}
   {rendererClass=&NiagaraRibbonRendererProperties, rendererIndex=1}
   {rendererClass=&NiagaraRibbonRendererProperties, rendererIndex=2}
   {rendererClass=&NiagaraRibbonRendererProperties, rendererIndex=3}
  ]
 }
}
summary {
 systemName=NS_Emote_SlickSwish_SingleHand_trail
 emitters [
  {
   emitterName=Hand_electricity001
   rendererClasses [&NiagaraMeshRendererProperties, &NiagaraLightRendererProperties]
  }
  {emitterName=Hand_electricity002_left_behind, rendererClasses [&NiagaraMeshRendererProperties]}
  {emitterName=hand_trail, rendererClasses [&NiagaraRibbonRendererProperties]}
 ]
 userVariables [
  User.Anim_Curve_ribbon_width {type=NiagaraFloat, defaultValue=1, description=""}
  User.Anim_Curve_SpawnRate_electricity {type=NiagaraFloat, defaultValue=1, description=""}
  User.Anim_Curve_SpawnRate_Ribbon {type=NiagaraFloat, defaultValue=1, description=""}
  User.bIsFrontEnd {type=NiagaraBool}
  User.bIsFrontEndPreview {type=NiagaraBool}
  User.RibbonIDOffset {type=NiagaraInt32}
  User.SK_mesh_hands {
   type=NiagaraDataInterfaceSkeletalMesh
   defaultValue {
    struct=NiagaraExt_VariableValue_DataInterface
    value {dataInterface=&self:NiagaraDataInterfaceSkeletalMesh_0}
   }
  }
 ]
}
