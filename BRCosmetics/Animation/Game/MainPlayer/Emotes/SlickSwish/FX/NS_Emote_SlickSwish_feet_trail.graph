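# Niagara system for the SlickSwish emote's feet trail, layering a short-lived magenta ribbon with flickering electricity meshes and pulsing lights.
# The singlehand emitter spawns ribbon particles that partly inherit source movement, with width and color over life and brightness scaled by time of day.
# The mesh_electricity emitter randomizes mesh index and scale, flickers alpha with waveforms and drives dynamic material parameters.
# Skeletal mesh sampling modules are disabled in both emitters.
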NiagaraSystem /BRCosmetics/Animation/Game/MainPlayer/Emotes/SlickSwish/FX/NS_Emote_SlickSwish_feet_trail
dependencies {
 usedDataInterfaces [
  &NiagaraDataInterfaceCurve
  &NiagaraDataInterfaceColorCurve
  &NiagaraDataInterfaceArrayFloat3
  &NiagaraDataInterfaceSkeletalMesh
 ]
 usedDynamicInputs []
 usedModules [
  &self:SystemSpawnScript.SystemScriptSource.SystemScriptGraph.NiagaraNodeAssignment_0.SetVariables_#1
  &SystemState
 ]
}
emitters {
 mesh_electricity {
  emitterUpdateScript [
   &EmitterState()
   &SpawnRate(SpawnRate=10, Use Spawn Probability=false)
  ]
  particleSpawnScript [
   &InitializeParticle(
    Color=<rgba 10 0.7227 4.7353 1>
    Lifetime Max=0.4
    Lifetime Min=0.2
    Lifetime Mode=::NewEnumerator1
    Mesh Scale=&RandomRangeVector(Minimum=<0.3 0.3 0.3>, Maximum=<0.5 0.5 0.45>)
    Mesh Scale Mode=::NewEnumerator4
    Sprite Size Mode=::NewEnumerator3
    Uniform Sprite Size=5
    UsePositionOffset=false
    Write Material Random=true
   )
   /SetVariables(Particles.Mesh index:NiagaraInt32=&RandomRangeInt(Minimum=0, Maximum=2)) {isSetParameters=true}
  ]
  particleUpdateScript [
   &ParticleState()
   &ShapeLocation(MaskSpawnGroup=false, Override Seed=false, Shape Origin=$Particles.SkeletalMeshLocation.SampledPosition, Sphere Radius=5, UseOwnerScale=true)
   &SkeletalMeshLocation(
    Bone / Socket Index=$Particles.PartitionParticles.Partition
    MaskSpawnGroup=false
    Mesh Sampling Type=::NewEnumerator5
    Orientation Sampling=::NewEnumerator1
    Sampled Position Offset=&SelectVectorFromArray(Direct Array Index=$Particles.PartitionParticles.Partition, Vector Selection Array=DataInterface() {propertyValues {FloatData []}})
    Skeletal Mesh=$System.Skeletal Mesh
    Socket Sampling Mode=::NewEnumerator5
   ) {enabled=false}
   &AlignSpriteToMeshOrientation(Mesh Orientation Relative Sprite Alignment Vector=<-1 0 0>, SetAlignment=true, SetFacing=false) {enabled=false}
   &ScaleColor(
    Scale Alpha=&Waveform(
     "Waveform [1]"=::NewEnumerator0
     "[1] Frequency"=&RandomRangeFloat(Minimum=5, Maximum=8)
     "[1] Amplitude Min/Max"=<0.5 1>
    )
    ScaleA=true
    ScaleRGB=true
   )
   &CameraOffset(Camera Offset Amount=10)
   /SetVariables(Particles.Glow color:LinearColor=&MakeLinearColorFromVectorAndFloat(
     "Vector (RGB)"=&Multiply_VectorByFloat(Vector=&MakeVectorFromLinearColorRGB(Color=<rgba 1 0.0723 0.4735 1>), Float=0.5)
     "Float (Alpha)"=&Multiply_Float(
      A=&Multiply_Float(
       A=&FloatFromCurve(
        FloatCurve=DataInterface() {propertyValues {Curve {keys [
            {interpMode=RCIM_Cubic, tangentMode=RCTM_User, time=0, value=0}
            {interpMode=RCIM_Cubic, tangentMode=RCTM_User, time=0.5, value=1}
            {tangentMode=RCTM_User, time=1, value=0}
           ]}}}
        Scale Curve=&FNTimeOfDayScalar(Night Scale=0.7, Morning Scale=0.3, Evening Scale=0.3)
       )
       B=&Multiply_Float(A=&RandomRangeFloat(Minimum=0, Maximum=1, Evaluation Type=::NewEnumerator1), B=1)
      )
      B=1
     )
    )) {isSetParameters=true}
   &DynamicMaterialParameters(
    Index 0 Param 1=&Multiply_Float(
     A=1.7
     B=&FloatFromCurve(FloatCurve=DataInterface() {propertyValues {Curve {keys [
          {arriveTangent=-2.6097, interpMode=RCIM_Cubic, leaveTangent=-2.6097, tangentMode=RCTM_User, time=0, value=1.5}
          {arriveTangent=-0.6794, interpMode=RCIM_Cubic, leaveTangent=-0.6794, tangentMode=RCTM_User, time=1, value=0}
         ]}}})
    )
    Index 0 Param 3=1
    Index 0 Param 4=1
    Module.Index0Float0Write=true
    Module.Index0Float1Write=true
    Module.Index0Float2Write=true
    Module.Index0Float3Write=true
   )
   &ScaleColor(
    Scale Alpha=&Multiply_Float(
     A=&FloatFromCurve(
      FloatCurve=DataInterface() {propertyValues {Curve {keys [
          {interpMode=RCIM_Cubic, tangentMode=RCTM_User, time=0, value=0}
          {interpMode=RCIM_Cubic, tangentMode=RCTM_User, time=0.5, value=1}
          {tangentMode=RCTM_User, time=1, value=0}
         ]}}}
      Scale Curve=&FNTimeOfDayScalar(Night Scale=0.7, Morning Scale=0.3, Evening Scale=0.3)
     )
     B=&Multiply_Float(A=&RandomRangeFloat(Minimum=0, Maximum=1, Evaluation Type=::NewEnumerator1), B=1)
    )
    ScaleA=true
    ScaleRGB=true
   ) {module=ScaleColor001}
   /SetVariables(Particles.Glow Dy02:Vector4f=&MakeVector4(
     X=&Multiply_Float(
      A=5
      B=&FloatFromCurve(FloatCurve=DataInterface() {propertyValues {Curve {keys [
           {arriveTangent=-2.6097, interpMode=RCIM_Cubic, leaveTangent=-2.6097, tangentMode=RCTM_User, time=0, value=1.5}
           {arriveTangent=-0.6794, interpMode=RCIM_Cubic, leaveTangent=-0.6794, tangentMode=RCTM_User, time=1, value=0}
          ]}}})
     )
     Y=0
     Z=0
     W=0
    )) {isSetParameters=true}
   /SetVariables(Particles.MeshOrientation=&MakeQuaternion(
     Quaternion From=5
     Axis=?
     Angle=?
     Yaw=?
     Pitch=?
     Roll=?
     X=?
     Y=?
     Z=?
     From=?
     To=?
     Angle Type=::NewEnumerator0
     Coordinate Space=?
     Normalize Vectors=?
    )) {isSetParameters=true}
   &CameraOffset(Camera Offset Amount=15) {module=CameraOffset001}
   &FN_Light_Attributes(
    Color=<rgba 0.6867 0 0.5054 1>
    Create Color Attribute=true
    Create Exponent Attribute=true
    Create Radius Attribute=true
    Enable Particle Light Enable=true
    Exponent=5
    Light Enabled=true
    Radius=&FloatFromCurve(
     FloatCurve=DataInterface() {propertyValues {Curve {keys [<at 0 1>,<at 0.1959 0.3345>,<at 0.4134 0.677>,<at 0.6308 0.2598>,<at 0.8385 0.8188>,<at 1 0>]}}}
     Scale Curve=978.6676
    )
    Scale Color=12.6888
   )
  ]
  rendererClasses [&NiagaraMeshRendererProperties, &NiagaraLightRendererProperties]
  renderers [
   {rendererClass=&NiagaraMeshRendererProperties, rendererIndex=0}
   {rendererClass=&NiagaraMeshRendererProperties, rendererIndex=1}
   {rendererClass=&NiagaraLightRendererProperties, rendererIndex=2}
  ]
 }
 singlehand {
  emitterUpdateScript [
   &EmitterState(Life Cycle Mode=::NewEnumerator1, Loop Duration=1, UseLoopDelay=false)
   &SpawnBurst_Instantaneous(Use Spawn Probability=false, UseLoopCountLimit=false) {enabled=false}
   &SpawnRate(SpawnRate=60, Use Spawn Probability=false)
  ]
  particleSpawnScript [
   &InitializeParticle(
    Color=<rgba 0.6867 0 0.5054 1>
    Lifetime=0.2
    Ribbon Twist Mode:ENiagara_UnsetDirectSet=::NewEnumerator1
    Ribbon Width=20
    Ribbon Width Mode=::NewEnumerator1
    UsePositionOffset=false
    Write Material Random=true
   )
   &PartitionParticles(Number Of Partitions=1) {enabled=false}
   &SkeletalMeshLocation(
    "Inherit Velocity (Scale)"=0.01
    MaskSpawnGroup=false
    Mesh Sampling Type=::NewEnumerator5
    OverrideRandomSeed=false
    Sampled Position Offset=&MaskVector(Mask Source=1, Mask Parameter=$Particles.PartitionParticles.Partition, Mask=0, Masked Vector=<0 -4.8987 0>, Passthrough Vector=<0 0 0>)
    Skeletal Mesh=$System.Skeletal Mesh
    Velocity Sampling=::NewEnumerator0
    VelocitySpeedLimit=false
   ) {enabled=false, module=SkeletalMeshLocation002}
   &AddVelocity(Velocity=<0 0 340.6738>) {enabled=false}
   &PointForce(
    Linear Force Vector:Vector3f=/Niagara/DynamicInputs/Helpers/MaskPosition(
     Mask Source:NiagaraInt32=0
     Mask Parameter:NiagaraInt32=$Particles.PartitionParticles.Partition
     Mask:NiagaraInt32=0
     Masked Position:NiagaraPosition=<1 0 0>
     Passthrough Position:NiagaraPosition=<-1 0 0>
    )
    Module.Use Falloff=true
    Module.UseLinearForce=true
   ) {enabled=false}
   &ApplyInitialForces()
  ]
  particleUpdateScript [
   &ParticleState()
   &InheritSourceMovement(Applied Position Scale=<0.5 0.5 0.5>, ApplyForce=false, ApplyPosition=true, ApplyPositionFalloff=false)
   &ScaleRibbonWidth(Ribbon Width Scale=&FloatFromCurve(FloatCurve=DataInterface() {propertyValues {Curve {
        defaultValue=1
        keys [
         {arriveTangent=3.5475, interpMode=RCIM_Cubic, leaveTangent=3.5475, tangentMode=RCTM_User, time=0, value=0.45}
         {arriveTangent=-0.1116, interpMode=RCIM_Cubic, leaveTangent=-0.1116, tangentMode=RCTM_User, time=0.3, value=1}
         {arriveTangent=-0.0405, interpMode=RCIM_Cubic, leaveTangent=-0.0405, tangentMode=RCTM_User, time=1, value=0}
        ]
       }}}))
   &ScaleColor(
    Scale Alpha=&FloatFromCurve(FloatCurve=DataInterface() {propertyValues {Curve {keys [{arriveTangent=-3.0806, interpMode=RCIM_Cubic, leaveTangent=-3.0806, tangentMode=RCTM_User, time=0, value=1}, <at 1 0 RCIM_Cubic>]}}})
    Scale RGB=&MakeVectorFromLinearColorRGB(Color=&ColorFromCurve(ColorCurve=DataInterface() {propertyValues {
        AlphaCurve {keys []}
        BlueCurve {keys [<at 0.3816 0.8698 RCIM_Cubic>,<at 1 1>]}
        GreenCurve {keys [<at 0.3816 0 RCIM_Cubic>,<at 1 0.1948>]}
        RedCurve {keys [<at 0.3816 1.5 RCIM_Cubic>,<at 1 0.9347>]}
       }}))
    ScaleA=true
    ScaleRGB=true
   ) {module=ScaleColor003}
   &ScaleColor(
    Scale Alpha=0.7
    Scale RGB=&VectorFromFloat(Value=&FNTimeOfDayScalar(Night Scale=25, Morning Scale=1, Day Scale=25, Evening Scale=1))
    ScaleA=false
    ScaleRGB=true
   ) {module=ScaleColor002}
   &SolveForcesAndVelocity(Clamp Velocity=false, Limit Acceleration=false)
   &DynamicMaterialParameters(Module.Index0Float0Write=true, Module.Index0Float1Write=true, Module.Index0Float2Write=true, Module.Index0Float3Write=true)
  ]
  rendererClasses [&NiagaraRibbonRendererProperties]
  renderers [
   {rendererClass=&NiagaraRibbonRendererProperties}
   {rendererClass=&NiagaraRibbonRendererProperties, rendererIndex=1}
  ]
 }
}
summary {
 systemName=NS_Emote_SlickSwish_feet_trail
 emitters [
  {emitterName=singlehand, rendererClasses [&NiagaraRibbonRendererProperties]}
  {
   emitterName=mesh_electricity
   rendererClasses [&NiagaraMeshRendererProperties, &NiagaraLightRendererProperties]
  }
 ]
 userVariables [
  User.bIsFrontEnd {type=NiagaraBool}
  User.bIsFrontEndPreview {type=NiagaraBool}
 ]
}
